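# AIGallery status widget that shows a follower icon, apparently marking an AI character that can be hired.
# A scaled-down vertical box holds the icon, over a background image that is collapsed by default.
# Construct and the EventHideStatus custom event both collapse the widget and broadcast the VisibilityChanged dispatcher with bVisible left false.
# The graph contains no logic that makes the widget visible.
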WidgetBlueprint /AIGallery/UI/AIGalleryWidget_CanBeHired
graphs [
 EventGraph {nodes [
   EventHideStatus() {is=title K2Node_CustomEvent}
   > SetVisibility(InVisibility=Collapsed)
   > Call Visibility Changed!(execute:Exec, self:AIGallery Widget Can Be Hired Object Reference, bVisible:bool) -> (then:Exec) {is=title K2Node_CallDelegate}
   Event Construct()
   > SetVisibility(InVisibility=Collapsed)
   > Call Visibility Changed()
  ]}
 VisibilityChanged {nodes [VisibilityChanged!() -> (then, bVisible:bool=~) {is=title K2Node_FunctionEntry}]}
]
variables [VisibilityChanged {}]
widgets {
 info {inheritedWidgetCount=0, parentClass=&UserWidget, rootWidgetClass=&Overlay, widgetCount=6}
 widgets [
  Overlay_12 {widgetClassPath=&Overlay, widget=Overlay_12}
  Overlay_0 {
   widgetClassPath=&Overlay
   parent=Overlay_12
   widget=Overlay_0
   slot {class=OverlaySlot, properties {horizontalAlignment=HAlign_Fill, verticalAlignment=VAlign_Bottom}}
  }
  Background {
   widgetClassPath=&Image
   bIsVariable=true
   parent=Overlay_0
   widget=Background
   properties {
    visibility=Collapsed
    brush {
     bIsDynamicallyLoaded=false
     drawAs=Image
     imageType=NoImage
     mirroring=NoMirror
     resourceName=None
     resourceObject=null
     tiling=NoTile
     imageSize=<161 18>
     margin=<ltrb 0 0 0 0>
     outlineSettings {
      bUseBrushTransparency=false
      roundingType=HalfHeightRadius
      width=0
      color {colorUseRule=UseColor_Specified, specifiedColor=<rgba 0 0 0 0>}
      cornerRadii=<0 0 0 0>
     }
     tintColor {colorUseRule=UseColor_Specified, specifiedColor=<rgba 1 1 1 1>}
     uVRegion {bIsValid=false, max=<0 0>, min=<0 0>}
    }
    colorAndOpacity=<rgba 0 0 0 0.32>
   }
   slot {class=OverlaySlot, properties {horizontalAlignment=HAlign_Center, verticalAlignment=VAlign_Center}}
  }
  VerticalBox_PlayerStatus {
   widgetClassPath=&VerticalBox
   parent=Overlay_0
   widget=VerticalBox_PlayerStatus
   properties {renderTransform {angle=0, scale=<0.75 0.75>, shear=<0 0>, translation=<0 0>}}
   slot {class=OverlaySlot}
  }
  SizeBox_0 {
   widgetClassPath=&SizeBox
   parent=VerticalBox_PlayerStatus
   widget=SizeBox_0
   properties {bOverride_HeightOverride=true, bOverride_MinDesiredWidth=true, heightOverride=64, minDesiredWidth=64}
   slot {class=VerticalBoxSlot, properties {verticalAlignment=VAlign_Top}}
  }
  FortLazyImage_31 {
   widgetClassPath=&FortLazyImage
   bIsVariable=true
   parent=SizeBox_0
   widget=FortLazyImage_31
   properties {brush {
     bIsDynamicallyLoaded=false
     drawAs=Image
     imageType=FullColor
     mirroring=NoMirror
     resourceName=None
     resourceObject=/FortniteCoreUI/HUD/MiniMap/Art/T_UI_Icon_Follower
     tiling=NoTile
     imageSize=<64 64>
     margin=<ltrb 0 0 0 0>
     outlineSettings {
      bUseBrushTransparency=false
      roundingType=HalfHeightRadius
      width=0
      color {colorUseRule=UseColor_Specified, specifiedColor=<rgba 0 0 0 0>}
      cornerRadii=<0 0 0 0>
     }
     tintColor {colorUseRule=UseColor_Specified, specifiedColor=<rgba 1 1 1 1>}
     uVRegion {bIsValid=false, max=<0 0>, min=<0 0>}
    }}
   slot {class=SizeBoxSlot}
  }
 ]
}
